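# Material function in the Landmass erosion brush that outputs LandscapeUVs for landscape render target coordinates.
# It adds the TexelCenter scalar parameter, a half-texel offset by default, to the texture coordinate.
# The sum is divided by the render target resolution, built from RTResX and RTResY in the MPC_Landscape parameter collection.
# The result passes through a named reroute, Landscape UVs, to the output.
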MaterialFunction /Landmass/Landscape/ErosionBrush/Materials/Functions/LandmassErosion_LandscapeUVs
graph {nodes [
  n1 = Add(A=TextureCoordinate(), B=ScalarParameter() {defaultValue=0.5, parameterName=TexelCenter})
  n2 = AppendVector(
   A=CollectionParameter() {collection=/Landmass/Landscape/BlueprintBrushes/MPC/MPC_Landscape, parameterName=RTResX}
   B=CollectionParameter() {collection=&MPC_Landscape, parameterName=RTResY}
  )
  FunctionOutput(None=NamedRerouteDeclaration(None=Divide(A=@n1, B=@n2)) {name=Landscape UVs, nodeColor=<rgba 1 0 0.2109 1>}) {outputName=LandscapeUVs}
 ]}
