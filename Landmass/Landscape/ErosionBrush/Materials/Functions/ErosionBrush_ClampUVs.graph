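# Material function for the Landmass erosion brush that clamps a 2D UV input to a valid sub-region of the render target.
# The lower bound is a half-texel offset derived from the resolution parameters in MPC_Erosion.
# The upper bound adds that same offset to the ratio of total size to render target world size, both read from the same parameter collection.
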MaterialFunction /Landmass/Landscape/ErosionBrush/Materials/Functions/ErosionBrush_ClampUVs
graph {nodes [
  n1 = Divide(B=AppendVector(
    A=CollectionParameter() {collection=/Landmass/Landscape/BlueprintBrushes/MPC/MPC_Erosion, parameterName=ResX}
    B=CollectionParameter() {collection=&MPC_Erosion, parameterName=ResY}
   )) {constA=0.5, constB=0.5}
  n2 = Divide(
   A=AppendVector(
    A=CollectionParameter() {collection=&MPC_Erosion, parameterName=TotalSizeX}
    B=CollectionParameter() {collection=&MPC_Erosion, parameterName=TotalSizeY}
   )
   B=AppendVector(
    A=CollectionParameter() {collection=&MPC_Erosion, parameterName=RTWorldSizeX}
    B=CollectionParameter() {collection=&MPC_Erosion, parameterName=RTWorldSizeY}
   )
  )
  FunctionOutput(None=Clamp(None=FunctionInput() {inputName=Add, inputType=FunctionInput_Vector2}, Min=@n1, Max=Add(A=@n1, B=@n2))) {outputName=Clamp}
 ]}
