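# Material function for the Landmass erosion brush that outputs particle UVs, particle world XY position and landscape UVs, using values from MPC_Erosion.
# A component index offset is found by flooring the transformed origin, relative to landscape location, by component size.
# Texture coordinates scaled by component resolution combine with offsets and are normalized by divisor-scaled landscape resolution.
# World XY scales by grid size and landscape resolution, shifted to landscape location minus a grid-cell fraction.
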MaterialFunction /Landmass/Landscape/ErosionBrush/Materials/Functions/ErosionBrush_ParticleCoords
graph {nodes [
  Divide(
   A=Constant() {r=2}
   B=CollectionParameter() {collection=/Landmass/Landscape/BlueprintBrushes/MPC/MPC_Erosion, parameterName=ResolutionDivisor}
  )
  n1 = Multiply(
   A=TextureCoordinate() {coordinateIndex=1}
   B=CollectionParameter() {collection=&MPC_Erosion, parameterName=ComponentRes}
  ) {desc=Particle Index + 0.5}
  n2 = Add(A=NamedRerouteUsage(), B=@n1)
  n3 = Divide(
   A=AppendVector(
    A=CollectionParameter() {collection=&MPC_Erosion, parameterName=ResX}
    B=CollectionParameter() {collection=&MPC_Erosion, parameterName=ResY}
   )
   B=CollectionParameter() {collection=&MPC_Erosion, parameterName=ResolutionDivisor}
  )
  FunctionOutput(None=NamedRerouteDeclaration(None=Divide(A=@n2, B=@n3)) {name=Particle Position 0-1, nodeColor=<rgba 0 1 0.0391 1>}) {outputName=Particle UVs}
  n4 = AppendVector(
   A=CollectionParameter() {collection=&MPC_Erosion, parameterName=ResX}
   B=CollectionParameter() {collection=&MPC_Erosion, parameterName=ResY}
  )
  n5 = CollectionParameter() {collection=&MPC_Erosion, parameterName=GridSize}
  n6 = Add(
   A=Multiply(A=NamedRerouteUsage(), B=Multiply(A=@n4, B=@n5))
   B=ComponentMask(None=CollectionParameter() {collection=&MPC_Erosion, parameterName=LandscapeLocation}) {g=true, r=true}
  )
  n7 = Divide(A=CollectionParameter() {collection=&MPC_Erosion, parameterName=GridSize})
  FunctionOutput(None=NamedRerouteDeclaration(None=Subtract(A=@n6, B=@n7)) {name=Particle World XY, nodeColor=<rgba 0.875 1 0 1>}) {outputName=Particle Position XY, sortPriority=1}
  n8 = TextureCoordinate()
  n9 = Divide(
   A=Subtract(A=CollectionParameter() {collection=&MPC_Erosion, parameterName=ComponentRes})
   B=CollectionParameter() {collection=&MPC_Erosion, parameterName=ComponentRes}
  )
  n10 = Add(A=Multiply(
    A=Multiply(A=@n8, B=@n9)
    B=CollectionParameter() {collection=&MPC_Erosion, parameterName=ComponentRes}
   )) {constB=0.5}
  n11 = NamedRerouteUsage()
  n12 = AppendVector(
   A=CollectionParameter() {collection=&MPC_Erosion, parameterName=ResX}
   B=CollectionParameter() {collection=&MPC_Erosion, parameterName=ResY}
  )
  n13 = CollectionParameter() {collection=&MPC_Erosion, parameterName=ResolutionDivisor}
  FunctionOutput(None=Divide(A=Add(A=@n10, B=@n11), B=Divide(A=@n12, B=@n13))) {outputName=Landscape UVs, sortPriority=1}
  Multiply(A=@n1, B=CollectionParameter() {collection=&MPC_Erosion, parameterName=ResolutionDivisor})
  n14 = TransformPosition(None=Constant3Vector()) {transformType=TRANSFORMPOSSOURCE_World}
  n15 = ShaderStageSwitch(PixelShader=VertexInterpolator(VS=@n14).PS, VertexShader=@n14)
  n16 = ComponentMask(None=CollectionParameter() {collection=&MPC_Erosion, parameterName=LandscapeLocation}) {b=true, g=true, r=true}
  n17 = CollectionParameter() {collection=&MPC_Erosion, parameterName=GridSize}
  n18 = Subtract(A=CollectionParameter() {collection=&MPC_Erosion, parameterName=ComponentRes}) {desc=Component Quads}
  n19 = Divide(
   A=ComponentMask(None=Subtract(A=@n15, B=@n16)) {g=true, r=true}
   B=Multiply(
    A=Multiply(A=@n17, B=@n18) {desc=Size Per Component}
    B=CollectionParameter() {collection=&MPC_Erosion, parameterName=ResolutionDivisor}
   )
  )
  n20 = CollectionParameter() {collection=&MPC_Erosion, parameterName=ComponentRes}
  NamedRerouteDeclaration(None=Multiply(A=Floor(None=@n19), B=Subtract(A=@n20) {desc=Component Quads})) {name=Vertex Index Offset, nodeColor=<rgba 0.7344 0 1 1>}
 ]}
